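# Physics asset for the Mockingbird backpack cosmetic, defining collision and joints for its dangling pockets, gun holders and straps, rope and hook strap.
# Three capsules form the pack body, which anchors every dynamic chain.
# Front pockets and gun holders swing and twist freely on the pack, while their child segments have angular limits.
# The rope is a limited three-link chain ending in three side-by-side capsules, and the hook strap is a single limited link.
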PhysicsAsset /BRCosmetics/Accessories/FORT_Backpacks/Backpack_F_MED_Mockingbird/Meshes/Mockingbird_PhysicsAsset
bodies {
 front_left_pocket_01_dyn [{shapeType=Capsule, length=8, radius=2.5, shapeName=front_left_pocket_01_dyn_capsule, center=<1 0 0>, rotation=<rot 0 -85 -50>}]
 front_left_pocket_02_dyn [{shapeType=Capsule, length=4.5, radius=0.8, shapeName=front_left_pocket_02_dyn_capsule, center=<2.25 0 0.25>, rotation=<rot 90 0 0>}]
 front_right_pocket_01_dyn [{shapeType=Capsule, length=8, radius=2.5, shapeName=front_right_pocket_01_dyn_capsule, center=<1 0 0>, rotation=<rot 6 -95 -50>}]
 front_right_pocket_02_dyn [{shapeType=Capsule, length=4.5, radius=0.8, shapeName=front_right_pocket_02_dyn_capsule, center=<2.25 0 0.25>, rotation=<rot 90 0 0>}]
 hook_strap_dyn [{shapeType=Capsule, length=6, radius=1, shapeName=hook_strap_dyn_capsule, center=<2 0 0>, rotation=<rot 0 0 90>}]
 left_gun_holder_dyn [{shapeType=Sphere, radius=1, shapeName=left_gun_holder_dyn_sphere}]
 left_gun_strap_dyn [{shapeType=Capsule, length=13, radius=2.75, shapeName=left_gun_strap_dyn_capsule, center=<-3 -0.75 0>, rotation=<rot 90 0 0>}]
 pack [
  {shapeType=Capsule, length=8, radius=8, center=<0.37 -7 -17>, rotation=<rot -90 0 0>}
  {shapeType=Capsule, length=14, radius=7, shapeName=pack_capsule, center=<0.4865 -6.3027 -2.2556>, rotation=<rot 90 0 0>}
  {shapeType=Capsule, length=13, radius=3.5, shapeName=pack_capsule, center=<0 -2.5 3.3159>, rotation=<rot 90 0 0>}
 ]
 right_gun_holder_dyn [{shapeType=Sphere, radius=1, shapeName=right_gun_holder_dyn_sphere}]
 right_gun_strap_dyn [{shapeType=Capsule, length=13, radius=2.75, shapeName=right_gun_strap_dyn_capsule, center=<-3 0 0>, rotation=<rot 90 0 0>}]
 rope_dyn_01 [{shapeType=Capsule, length=3, radius=1, shapeName=rope_dyn_01_capsule, center=<3 0 0>, rotation=<rot 90 0 0>}]
 rope_dyn_02 [{shapeType=Capsule, length=3, radius=1, shapeName=rope_dyn_02_capsule, center=<3 0 0>, rotation=<rot 90 0 0>}]
 rope_dyn_03 [
  {shapeType=Capsule, length=3, radius=1, shapeName=rope_dyn_03_capsule, center=<1.5 0 0>, rotation=<rot 90 0 0>}
  {shapeType=Capsule, length=3, radius=1, shapeName=rope_dyn_03_capsule, center=<1.5 2 0>, rotation=<rot 90 0 0>}
  {shapeType=Capsule, length=3, radius=1, shapeName=rope_dyn_03_capsule, center=<1.5 -2 0>, rotation=<rot 90 0 0>}
 ]
}
constraints [
 {
  bone1Name=front_left_pocket_01_dyn
  bone2Name=pack
  swing1LimitDegrees=45
  swing1Motion=Free
  swing2LimitDegrees=45
  swing2Motion=Free
  twistLimitDegrees=45
  twistMotion=Free
 }
 {bone1Name=front_left_pocket_02_dyn, bone2Name=front_left_pocket_01_dyn, swing1LimitDegrees=4, swing2LimitDegrees=45, twistLimitDegrees=5}
 {
  bone1Name=front_right_pocket_01_dyn
  bone2Name=pack
  swing1LimitDegrees=45
  swing1Motion=Free
  swing2LimitDegrees=45
  swing2Motion=Free
  twistLimitDegrees=45
  twistMotion=Free
 }
 {bone1Name=front_right_pocket_02_dyn, bone2Name=front_right_pocket_01_dyn, swing1LimitDegrees=2, swing2LimitDegrees=25, twistLimitDegrees=5}
 {
  bone1Name=right_gun_holder_dyn
  bone2Name=pack
  swing1LimitDegrees=45
  swing1Motion=Free
  swing2LimitDegrees=45
  swing2Motion=Free
  twistLimitDegrees=45
  twistMotion=Free
 }
 {bone1Name=right_gun_strap_dyn, bone2Name=right_gun_holder_dyn, swing1LimitDegrees=2, swing2LimitDegrees=3, twistLimitDegrees=1}
 {
  bone1Name=left_gun_holder_dyn
  bone2Name=pack
  swing1LimitDegrees=45
  swing1Motion=Free
  swing2LimitDegrees=45
  swing2Motion=Free
  twistLimitDegrees=45
  twistMotion=Free
 }
 {bone1Name=left_gun_strap_dyn, bone2Name=left_gun_holder_dyn, swing1LimitDegrees=2, swing2LimitDegrees=3, twistLimitDegrees=1}
 {bone1Name=rope_dyn_01, bone2Name=pack, swing1LimitDegrees=2, swing2LimitDegrees=20, twistLimitDegrees=1}
 {bone1Name=rope_dyn_02, bone2Name=rope_dyn_01, swing1LimitDegrees=2, swing2LimitDegrees=20, twistLimitDegrees=1}
 {bone1Name=rope_dyn_03, bone2Name=rope_dyn_02, swing1LimitDegrees=2, swing2LimitDegrees=15, twistLimitDegrees=1}
 {bone1Name=hook_strap_dyn, bone2Name=pack, swing1LimitDegrees=15, swing2LimitDegrees=45, twistLimitDegrees=5}
]
